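# Unlit translucent master material for Niagara sprite and mesh particles that draws comic-style Ben-Day halftone dots.
# Screen-space UVs anchor to object position and scale with view depth, then tile into thresholded radial dots.
# Optional masks size the dots: a pixelated texture cutout with dynamic-parameter resolution, and a quantized, object-stabilized radial gradient.
# Particle color tints the pattern through emissive adjustments, while particle alpha and near-camera fade drive opacity.
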Material /BRCosmetics/Effects/Niagara/MasterMaterials/ToonShader/M_BenDay_Dots_FX
graph {
 properties {
  bAutomaticallySetUsageInEditor=false
  bUsedWithNiagaraMeshParticles=true
  bUsedWithNiagaraSprites=true
  blendMode=BLEND_Translucent
  pixelDepthOffsetMode=PDOM_Legacy
  shadingModel=MSM_Unlit
  usedShadingModels=Unlit
 }
 nodes [
  n1 = ConstantBiasScale(None=ScreenPosition().ViewportUV) {bias=-0.5, scale=2}
  n2 = MaterialFunctionCall(Float2=ViewSize()) {materialFunction=&BreakOutFloat2Components}
  n3 = AppendVector(A=Constant() {r=1}, B=Divide(A=@n2.G, B=@n2.G))
  n4 = Multiply(
   A=MaterialFunctionCall(World Position=ObjectPositionWS()) {materialFunction=&TransformToClipSpace}.Clip Space XY
   B=Constant2Vector() {g=-1, r=1}
  )
  n5 = Subtract(A=ConstantBiasScale(None=Multiply(A=@n1, B=@n3)), B=Multiply(A=Multiply(A=@n3, B=@n4)) {constB=0.5})
  n6 = TransformPosition(None=ObjectPositionWS()) {transformSourceType=TRANSFORMPOSSOURCE_World, transformType=TRANSFORMPOSSOURCE_View}
  n7 = Multiply(A=Divide(B=ComponentMask(None=@n6) {b=true}))
  n8 = Frac(None=Multiply(
    A=MaterialFunctionCall(UVs=@n5, Texture Scale=@n7) {materialFunction=&ScaleUVsByCenter}.UVs
    B=ScalarParameter() {defaultValue=1, parameterName=Tiling}
   ))
  n9 = SmoothStep(Value=Length(None=Subtract(A=@n8) {constB=0.5}) {is=class MaterialExpressionLength}) {constMax=0.5}
  n10 = TextureCoordinate()
  n11 = LinearInterpolate(
   A=ScalarParameter() {defaultValue=32, group=TextureCutout, parameterName=TextureResolution}
   B=DynamicParameter() {defaultValue=<rgba 32 1 1 1>, paramNames [Texture Resolution, Param2, Param3, Param4]}.Texture Resolution
   Alpha=ScalarParameter() {group=TextureCutout, parameterName=UseDynamicParam?}
  )
  n12 = Divide(A=Floor(None=Multiply(A=@n10, B=@n11) {constB=10}), B=@n11) {constB=10}
  n13 = Step(X=ChannelMaskParameter(None=TextureSampleParameter2D(UVs=@n12) {group=TextureCutout, parameterName=TextureCutoutInput, texture=/Game/Effects/Fort_Effects/Textures/T_PillowMill_Shapes}.RGBA) {group=TextureCutout, parameterName=TextureChannelSelect}.R) {constY=0.8}
  n14 = LinearInterpolate(
   A=Step(Y=@n9, X=ScalarParameter() {defaultValue=0.5, parameterName=DotSize}) {constX=0.5}
   B=Step(Y=@n9, X=LinearInterpolate(Alpha=@n13) {constB=0.5}) {constX=0.5}
   Alpha=ScalarParameter() {group=TextureCutout, parameterName=UseTextureCutout?}
  ) {constA=0.5}
  n15 = ScalarParameter() {defaultValue=7, group=BenDayDots, parameterName=SphereMaskResolution}
  n16 = Multiply(
   A=MaterialFunctionCall!(Position=ObjectPositionWS(), UVRadialScale=ScalarParameter() {defaultValue=0.25, group=BenDayDots, parameterName=SphereMaskRadius}) {materialFunction=/Game/Weapons/WeaponSkins/MaterialFunction/ObjectStabilizedScreenSpaceUVs}.StabilizedUVs
   B=@n15
  ) {constB=20}
  n17 = MaterialFunctionCall!(UVs=Divide(A=Floor(None=@n16), B=@n15) {constB=20}, CenterPosition, Radius) {materialFunction=/Engine/Functions/Engine_MaterialFunctions01/Gradient/SphereGradient-2D}
  n18 = Add(
   A=ScalarParameter() {defaultValue=0.4, group=BenDayDots, parameterName=DotScaleMaster}
   B=LinearInterpolate(
    A=ScalarParameter() {group=BenDayDots, parameterName=SphereMaskDotScaleRange_Min}
    B=ScalarParameter() {defaultValue=0.6, group=BenDayDots, parameterName=SphereMaskDotScaleRange_Max}
    Alpha=Power(Base=@n17.Result 0-1)
   ) {constB=0.6}
  ) {constA=0.2}
  n19 = LinearInterpolate(A=@n14, B=Multiply(A=Step(Y=@n9, X=@n18), B=@n14), Alpha=ScalarParameter() {group=BenDayDots, parameterName=UseBenDayMask?})
  n20 = ParticleColor()
  Reroute(None=MaterialFunctionCall(Emissive=Multiply(A=@n19, B=@n20.RGB)) {materialFunction=&MF_FX_EmissiveAdjustments}.Result)
  n21 = ScalarParameter() {defaultValue=150, parameterName=CameraFadeDistance}
  Reroute(None=Multiply(
    A=Multiply(A=@n19, B=@n20.A)
    B=MaterialFunctionCall(Distance=@n21) {materialFunction=&Near_Camera_Fade}.Result
   ))
 ]
}
